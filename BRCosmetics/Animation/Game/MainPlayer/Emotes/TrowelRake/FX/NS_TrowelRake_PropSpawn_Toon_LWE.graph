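# Toon-style Niagara burst for the prop spawn in the TrowelRake player emote, with airglow, sparkle, smoke, rock debris and speed line emitters.
# Only Speed_Lines is detailed here: it briefly spawns short-lived grey sprites in a box, slows them with drag, reshapes them and fades their alpha.
# Its camera offset switches between gameplay and front-end values.
# A 1mCube_Delete emitter is disabled.
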NiagaraSystem /BRCosmetics/Animation/Game/MainPlayer/Emotes/TrowelRake/FX/NS_TrowelRake_PropSpawn_Toon_LWE
dependencies {
 usedDataInterfaces [&NiagaraDataInterfaceCurve, &NiagaraDataInterfaceVector2DCurve, &NiagaraDataInterfaceArrayFloat]
 usedDynamicInputs [
  &MakeFloatFromVector2d
  &SelectFloatFromArray
  &MakeIntFromFloat
  &FNTimeOfDayColor
 ]
 usedModules [
  &SystemState
  &self:SystemSpawnScript.SystemScriptSource.SystemScriptGraph.NiagaraNodeAssignment_0.SetVariables_#1
 ]
}
emitters {
 1mCube_Delete {bEnabled=false, rendererClasses [], renderers []}
 AirglowLWE {rendererClasses [], renderers []}
 Rock_Debris {rendererClasses [], renderers []}
 SparklesLWE {rendererClasses [], renderers []}
 Speed_Lines {
  emitterUpdateScript [
   &EmitterState(Life Cycle Mode=::NewEnumerator1, Loop Behavior=::NewEnumerator1, Loop Duration=0.2, UseLoopDelay=false)
   &SpawnRate(SpawnRate=50, Use Spawn Probability=false) {module=SpawnRate001}
  ]
  particleSpawnScript [
   &InitializeParticle(
    Color=&RandomRangeLinearColor(Minimum=<rgba 0.0365 0.0365 0.0365 10>, Maximum=<rgba 0.2573 0.2573 0.2573 15>)
    Lifetime=0.1
    Position Mode=::NewEnumerator1
    Sprite Size Max=<66.6667 180>
    Sprite Size Min=<30 125>
    Sprite Size Mode=::NewEnumerator2
    UsePositionOffset=false
    Write Material Random=true
   )
   &ShapeLocation(
    Box Size=<50 50 50>
    MaskSpawnGroup=false
    Offset Coordinate Space=::Local
    Offset Mode=::NewEnumerator0
    Override Seed=false
    Rotation Mode=::NewEnumerator2
    Shape Primitive=::NewEnumerator2
    Surface Only=false
    UseOwnerScale=true
    Yaw / Pitch / Roll=<0.16 0 0>
   )
   &AddVelocity() {module=AddVelocity001}
   &CameraOffset(Camera Offset Amount=&FN_FrontEndFloatModifier(Gameplay=0, Front End=-30))
  ]
  particleUpdateScript [
   &ParticleState()
   &Drag(Drag=5, Use Linear Drag=true, Use Rotational Drag=false)
   &SolveForcesAndVelocity(Clamp Velocity=false, Limit Acceleration=false)
   &ScaleColor(
    Scale Alpha=&FloatFromCurve(FloatCurve=DataInterface() {propertyValues {Curve {keys [
         {interpMode=RCIM_Cubic, tangentMode=RCTM_User, time=0.1552, value=0}
         {interpMode=RCIM_Cubic, tangentMode=RCTM_User, time=0.5, value=1}
         {tangentMode=RCTM_User, time=1, value=0}
        ]}}})
    ScaleA=true
    ScaleRGB=true
   )
   &ScaleSpriteSize(
    Non-Uniform Curve Sprite Scale=DataInterface() {propertyValues {
      XCurve {keys [{tangentMode=RCTM_User, time=0, value=0.5}, {tangentMode=RCTM_User, time=1, value=1.5}]}
      YCurve {keys [{tangentMode=RCTM_User, time=0, value=1}, {tangentMode=RCTM_User, time=1, value=0.5}]}
     }}
    Scale Sprite Size Mode=::NewEnumerator3
   )
   &ScaleSpriteSize() {module=ScaleSpriteSize001}
   /SetVariables(Particles.New Position:NiagaraPosition=&Add_Vector(
     A=Linked() {linkedVariable {name=Particles.Position, type=NiagaraPosition}}
     B=Linked() {linkedVariable {name=Engine.Emitter.SimulationPosition, type=NiagaraPosition}}
    )) {enabled=false, isSetParameters=true}
  ]
  rendererClasses [&NiagaraSpriteRendererProperties]
  renderers [{rendererClass=&NiagaraSpriteRendererProperties}]
 }
 ToonImpactSmoke {rendererClasses [], renderers []}
 ToonSmoke {rendererClasses [], renderers []}
}
summary {
 systemName=NS_TrowelRake_PropSpawn_Toon_LWE
 emitters [
  {emitterName=AirglowLWE, rendererClasses []}
  {emitterName=SparklesLWE, rendererClasses []}
  {emitterName=ToonImpactSmoke, rendererClasses []}
  {emitterName=ToonSmoke, rendererClasses []}
  {emitterName=Rock_Debris, rendererClasses []}
  {emitterName=Speed_Lines, rendererClasses [&NiagaraSpriteRendererProperties]}
  {bEnabled=false, emitterName=1mCube_Delete, rendererClasses []}
 ]
 userVariables [
  User.bIsFrontEnd {type=NiagaraBool}
  User.bIsFrontEndPreview {type=NiagaraBool}
  User.DesertBoulder {type=NiagaraBool}
  User.NormalizedNotifyProgress {type=NiagaraFloat, defaultValue=0, description=""}
  User.WorkoutSelect {type=NiagaraFloat, defaultValue=0, description=""}
 ]
}
