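# Niagara system for the TrowelRake emote that draws short-lived ribbon trails circling a ring, apparently as the sign's spawn effect.
# Three emitters stack vertically, each sweeping its ring spawn point by emitter age times User.ReverseTimeNS, with the top one time-shifted and the middle one faster and wider.
# Time-of-day color tints all trails.
# Top and bottom fade to black and use two ribbon renderers, with extra dark and blur width attributes.
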NiagaraSystem /BRCosmetics/Animation/Game/MainPlayer/Emotes/TrowelRake/FX/NS_TrowelRake_Trails_Sign_RingSpawn
dependencies {usedDataInterfaces [&NiagaraDataInterfaceCurve], usedDynamicInputs [], usedModules [&SystemState]}
emitters {
 bottom001 {
  emitterUpdateScript [
   &EmitterState(Inactive Response=::NewEnumerator1, Life Cycle Mode=::NewEnumerator1, UseLoopDelay=false)
   &SpawnRate(SpawnRate=100, Use Spawn Probability=false)
   &SpawnPerUnit(
    Max Movement Threshold=10000
    Module.Use Max Movement Threshold=true
    Module.Use Movement Tolerance=true
    Movement Tolerance=1
    Spawn Spacing=30
    Use Spawn Probability=false
   ) {enabled=false}
  ]
  particleSpawnScript [
   &InitializeParticle(
    Color=<rgba 0.25 0.25 0.25 1>
    Lifetime=0.07
    Position Offset=<0 0 -46>
    Ribbon Width=7.5
    Ribbon Width Mode=::NewEnumerator1
    UsePositionOffset=true
    Write Material Random=true
   )
   &ShapeLocation(
    MaskSpawnGroup=false
    Override Seed=false
    Radius Position=40
    Ring / Disc Distribution Mode:ENiagaraSphereDistributionMode=::NewEnumerator1
    Ring Radius=1
    Shape Primitive=::NewEnumerator4
    U Position=&Multiply_Float(
     A=&Subtract_Float(A=$Emitter.Age, B=0)
     B=&Multiply_Float(A=$User.ReverseTimeNS, B=7)
    )
    UseOwnerScale=true
   )
  ]
  particleUpdateScript [
   &ParticleState()
   &ScaleColor(
    Scale Alpha=&FloatFromCurve(FloatCurve=DataInterface() {propertyValues {Curve {keys [<at 0 0>,<at 0.5 1>,<at 1 0>]}}})
    Scale RGB=&Lerp_Vector(
     A=&MakeVectorFromLinearColorRGB(Color=&FNTimeOfDayColor(Night Color=<rgba 0.4168 0.5702 0.774 1>, Morning Color=<rgba 1 0.6827 0.6094 1>, Day Color=<rgba 1 1 1 1>, Evening Color=<rgba 1 1 1 1>))
     B=<0 0 0>
     Alpha=$Particles.NormalizedAge
    )
    ScaleA=true
    ScaleRGB=true
   )
   /SetVariables(Particles.RibbonWidth_Blur=20) {isSetParameters=true, module=SetVariables_#1}
   /SetVariables(Particles.ColorDark:LinearColor=<rgba 0.07 0.07 0.07 1>, Particles.RibbonWidthDark=3) {isSetParameters=true}
  ]
  rendererClasses [&NiagaraRibbonRendererProperties]
  renderers [
   {rendererClass=&NiagaraRibbonRendererProperties}
   {rendererClass=&NiagaraRibbonRendererProperties, rendererIndex=1}
  ]
 }
 middle {
  emitterUpdateScript [
   &EmitterState(Inactive Response=::NewEnumerator1, Life Cycle Mode=::NewEnumerator1, UseLoopDelay=false)
   &SpawnRate(SpawnRate=180, Use Spawn Probability=false)
   &SpawnPerUnit(
    Max Movement Threshold=10000
    Module.Use Max Movement Threshold=true
    Module.Use Movement Tolerance=true
    Movement Tolerance=1
    Spawn Spacing=30
    Use Spawn Probability=false
   ) {enabled=false}
  ]
  particleSpawnScript [
   &InitializeParticle(
    Color=<rgba 0.2 0.1255 0.025 1>
    Lifetime=0.025
    Position Offset=<0 0 -20>
    Ribbon Facing Mode=::NewEnumerator1
    Ribbon Width=40
    Ribbon Width Mode=::NewEnumerator1
    UsePositionOffset=true
    Write Material Random=true
   )
   &ShapeLocation(
    MaskSpawnGroup=false
    Override Seed=false
    Radius Position=39
    Ring / Disc Distribution Mode:ENiagaraSphereDistributionMode=::NewEnumerator1
    Ring Radius=1
    Shape Primitive=::NewEnumerator4
    U Position=&Multiply_Float(
     A=&Subtract_Float(A=$Emitter.Age, B=0)
     B=&Multiply_Float(A=$User.ReverseTimeNS, B=10)
    )
    UseOwnerScale=true
   )
  ]
  particleUpdateScript [
   &ParticleState()
   &ScaleColor(
    Scale Alpha=&FloatFromCurve(FloatCurve=DataInterface() {propertyValues {Curve {keys [<at 0 0>,<at 0.1 1>,<at 1 0>]}}})
    Scale RGB=&MakeVectorFromLinearColorRGB(Color=&FNTimeOfDayColor(Night Color=<rgba 0.4168 0.5702 0.774 1>, Morning Color=<rgba 1 0.6827 0.6094 1>, Day Color=<rgba 1 1 1 1>, Evening Color=<rgba 1 1 1 1>))
    ScaleA=true
    ScaleRGB=true
   )
   /SetVariables(Particles.RibbonWidth_Blur=20) {enabled=false, isSetParameters=true, module=SetVariables_#1}
  ]
  rendererClasses [&NiagaraRibbonRendererProperties]
  renderers [{rendererClass=&NiagaraRibbonRendererProperties}]
 }
 top001 {
  emitterUpdateScript [
   &EmitterState(Inactive Response=::NewEnumerator1, Life Cycle Mode=::NewEnumerator1, UseLoopDelay=false)
   &SpawnRate(SpawnRate=100, Use Spawn Probability=false)
   &SpawnPerUnit(
    Max Movement Threshold=10000
    Module.Use Max Movement Threshold=true
    Module.Use Movement Tolerance=true
    Movement Tolerance=1
    Spawn Spacing=30
    Use Spawn Probability=false
   ) {enabled=false}
  ]
  particleSpawnScript [
   &InitializeParticle(
    Color=<rgba 0.25 0.25 0.25 1>
    Lifetime=0.07
    Ribbon Width=7.5
    Ribbon Width Mode=::NewEnumerator1
    UsePositionOffset=false
    Write Material Random=true
   )
   &ShapeLocation(
    MaskSpawnGroup=false
    Override Seed=false
    Radius Position=40
    Ring / Disc Distribution Mode:ENiagaraSphereDistributionMode=::NewEnumerator1
    Ring Radius=1
    Shape Primitive=::NewEnumerator4
    U Position=&Multiply_Float(
     A=&Subtract_Float(A=$Emitter.Age, B=0.5)
     B=&Multiply_Float(A=$User.ReverseTimeNS, B=7)
    )
    UseOwnerScale=true
   )
  ]
  particleUpdateScript [
   &ParticleState()
   &ScaleColor(
    Scale Alpha=&FloatFromCurve(FloatCurve=DataInterface() {propertyValues {Curve {keys [<at 0 0>,<at 0.5 1>,<at 1 0>]}}})
    Scale RGB=&Lerp_Vector(
     A=&MakeVectorFromLinearColorRGB(Color=&FNTimeOfDayColor(Night Color=<rgba 0.4168 0.5702 0.774 1>, Morning Color=<rgba 1 0.6827 0.6094 1>, Day Color=<rgba 1 1 1 1>, Evening Color=<rgba 1 1 1 1>))
     B=<0 0 0>
     Alpha=$Particles.NormalizedAge
    )
    ScaleA=true
    ScaleRGB=true
   )
   /SetVariables(Particles.RibbonWidth_Blur=10) {isSetParameters=true, module=SetVariables_#1}
   /SetVariables(Particles.ColorDark:LinearColor=<rgba 0.07 0.07 0.07 1>, Particles.RibbonWidthDark=3) {isSetParameters=true}
  ]
  rendererClasses [&NiagaraRibbonRendererProperties]
  renderers [
   {rendererClass=&NiagaraRibbonRendererProperties}
   {rendererClass=&NiagaraRibbonRendererProperties, rendererIndex=1}
  ]
 }
}
summary {
 systemName=NS_TrowelRake_Trails_Sign_RingSpawn
 emitters [
  {emitterName=middle, rendererClasses [&NiagaraRibbonRendererProperties]}
  {emitterName=top001, rendererClasses [&NiagaraRibbonRendererProperties]}
  {emitterName=bottom001, rendererClasses [&NiagaraRibbonRendererProperties]}
 ]
 userVariables [User.ReverseTimeNS {type=NiagaraFloat, defaultValue=1, description=""}]
}
