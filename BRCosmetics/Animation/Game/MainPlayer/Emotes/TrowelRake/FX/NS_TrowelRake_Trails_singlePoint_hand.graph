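# Niagara system for the TrowelRake emote that draws short motion-blur ribbon trails, named for a single point on the hand.
# Two distance-spawned emitters, trails_blur_long and trails_blur_short, differ in particle lifetime and ribbon width, with the short one briefer and wider.
# Color is tinted by time of day and fades to black and transparent over each particle's life.
# Each emitter has two ribbon renderers and also writes separate dark color and width attributes.
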NiagaraSystem /BRCosmetics/Animation/Game/MainPlayer/Emotes/TrowelRake/FX/NS_TrowelRake_Trails_singlePoint_hand
dependencies {usedDataInterfaces [&NiagaraDataInterfaceCurve], usedDynamicInputs [], usedModules [&SystemState]}
emitters {
 trails_blur_long {
  emitterUpdateScript [
   &EmitterState(Inactive Response=::NewEnumerator1, Life Cycle Mode=::NewEnumerator1, Loop Behavior=::NewEnumerator1, Loop Duration=1, UseLoopDelay=false)
   &SpawnPerUnit(Module.Use Max Movement Threshold=false, Module.Use Movement Tolerance=false, Use Spawn Probability=false)
  ]
  particleSpawnScript [&InitializeParticle(
    Color=<rgba 0.25 0.25 0.25 1>
    Lifetime=0.1
    Ribbon Width=10
    Ribbon Width Mode=::NewEnumerator1
    UsePositionOffset=false
    Write Material Random=true
   )]
  particleUpdateScript [
   &ParticleState()
   &ScaleColor(
    Scale Alpha=&FloatFromCurve(FloatCurve=DataInterface() {propertyValues {Curve {keys [<at 0.0015 1>,<at 1 0>]}}})
    Scale RGB=&Lerp_Vector(
     A=&MakeVectorFromLinearColorRGB(Color=&FNTimeOfDayColor(Night Color=<rgba 0.4168 0.5702 0.774 1>, Morning Color=<rgba 1 0.6827 0.6094 1>, Day Color=<rgba 1 1 1 1>, Evening Color=<rgba 1 1 1 1>))
     B=<0 0 0>
     Alpha=$Particles.NormalizedAge
    )
    ScaleA=true
    ScaleRGB=true
   )
   /SetVariables(Particles.ColorDark:LinearColor=<rgba 0.07 0.07 0.07 1>, Particles.RibbonWidthDark=5) {isSetParameters=true, module=SetVariables_#1}
  ]
  rendererClasses [&NiagaraRibbonRendererProperties]
  renderers [
   {rendererClass=&NiagaraRibbonRendererProperties}
   {rendererClass=&NiagaraRibbonRendererProperties, rendererIndex=1}
  ]
 }
 trails_blur_short {
  emitterUpdateScript [
   &EmitterState(Inactive Response=::NewEnumerator1, Life Cycle Mode=::NewEnumerator1, Loop Behavior=::NewEnumerator1, Loop Duration=1, UseLoopDelay=false)
   &SpawnPerUnit(Module.Use Max Movement Threshold=false, Module.Use Movement Tolerance=false, Spawn Spacing=5, Use Spawn Probability=false)
  ]
  particleSpawnScript [&InitializeParticle(
    Color=<rgba 0.25 0.25 0.25 1>
    Lifetime=0.06
    Ribbon Width=15
    Ribbon Width Mode=::NewEnumerator1
    UsePositionOffset=false
    Write Material Random=true
   )]
  particleUpdateScript [
   &ParticleState()
   &ScaleColor(
    Scale Alpha=&FloatFromCurve(FloatCurve=DataInterface() {propertyValues {Curve {keys [<at 0.0015 1>,<at 1 0>]}}})
    Scale RGB=&Lerp_Vector(
     A=&MakeVectorFromLinearColorRGB(Color=&FNTimeOfDayColor(Night Color=<rgba 0.4168 0.5702 0.774 1>, Morning Color=<rgba 1 0.6827 0.6094 1>, Day Color=<rgba 1 1 1 1>, Evening Color=<rgba 1 1 1 1>))
     B=<0 0 0>
     Alpha=$Particles.NormalizedAge
    )
    ScaleA=true
    ScaleRGB=true
   )
   /SetVariables(Particles.ColorDark:LinearColor=<rgba 0.07 0.07 0.07 1>, Particles.RibbonWidthDark=5) {isSetParameters=true, module=SetVariables_#1}
  ]
  rendererClasses [&NiagaraRibbonRendererProperties]
  renderers [
   {rendererClass=&NiagaraRibbonRendererProperties}
   {rendererClass=&NiagaraRibbonRendererProperties, rendererIndex=1}
  ]
 }
}
summary {
 systemName=NS_TrowelRake_Trails_singlePoint_hand
 emitters [
  {emitterName=trails_blur_short, rendererClasses [&NiagaraRibbonRendererProperties]}
  {emitterName=trails_blur_long, rendererClasses [&NiagaraRibbonRendererProperties]}
 ]
 userVariables []
}
